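# Physics asset for the HotDog pickaxe, a two-handed melee cosmetic, giving it collision bodies and joints for slight secondary motion.
# It has three capsule bodies: a long one on Melee_TwoHanded_Handle and shorter ones on the dyn_top and dyn_bot bones.
# Both dyn bodies constrain directly to the handle rather than to each other, with tight swing limits and even tighter twist, keeping any wobble small.
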PhysicsAsset /BRCosmetics/Weapons/FORT_Melee/Pickaxe_HotDog/Meshes/Hotdog_Axe_Physics
bodies {
 Melee_TwoHanded_Handle [{shapeType=Capsule, length=27.5323, radius=4.774, center=<1.9582 -3.1479 25.4047>, rotation=<rot -0.278 -90 1.6587>}]
 dyn_bot [{shapeType=Capsule, length=6, radius=5.25, center=<1.9438 0.0004 0.1063>, rotation=<rot 0 -89.9999 -93.1306>}]
 dyn_top [{shapeType=Capsule, length=7.75, radius=4.5, center=<3.26 -0.0007 -0.0717>, rotation=<rot 0 -89.9998 91.2597>}]
}
constraints [
 {bone1Name=dyn_top, bone2Name=Melee_TwoHanded_Handle, swing1LimitDegrees=5, swing2LimitDegrees=5, twistLimitDegrees=1}
 {bone1Name=dyn_bot, bone2Name=Melee_TwoHanded_Handle, swing1LimitDegrees=5, swing2LimitDegrees=5, twistLimitDegrees=1}
]
